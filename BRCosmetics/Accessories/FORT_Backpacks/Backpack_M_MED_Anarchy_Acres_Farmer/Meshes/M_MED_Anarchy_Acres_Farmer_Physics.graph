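# Physics asset for the Anarchy Acres Farmer backpack cosmetic, simulating three straw pieces: front, middle and back.
# Each piece is a chain of four box bodies linked in series from its own master bone body, with no constraints between chains.
# All joints share tight angular limits that favor one swing axis, so the straw bends slightly in mostly one plane with almost no twist.
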PhysicsAsset /BRCosmetics/Accessories/FORT_Backpacks/Backpack_M_MED_Anarchy_Acres_Farmer/Meshes/M_MED_Anarchy_Acres_Farmer_Physics
bodies {
 strawBack1 [{shapeType=Box, extentX=2.01, extentY=1.01, extentZ=1.01, center=<1.0473 0.0001 -0.0022>}]
 strawBack2 [{shapeType=Box, extentX=2.51, extentY=1.01, extentZ=1.01, center=<1.3612 0 0.0006>}]
 strawBack3 [{shapeType=Box, extentX=1.51, extentY=1.01, extentZ=1.01, center=<0.7434 -0.0013 -0.0078>}]
 strawBack4 [{shapeType=Box, extentX=5.8676, extentY=1.0042, extentZ=1.2837, center=<3.0365 0.0077 -0.1513>}]
 strawBack_mstr [{shapeType=Box, extentX=1.01, extentY=1.01, extentZ=1.01}]
 strawFront1 [{shapeType=Box, extentX=1.51, extentY=1.01, extentZ=1.01, center=<0.5937 0.0037 -0.0352>}]
 strawFront2 [{shapeType=Box, extentX=2.01, extentY=1.01, extentZ=1.01, center=<1.1607 -0.0018 -0.0197>}]
 strawFront3 [{shapeType=Box, extentX=2.01, extentY=1.01, extentZ=1.01, center=<1.1161 0.0138 -0.0017>}]
 strawFront4 [{shapeType=Box, extentX=3.8481, extentY=0.6362, extentZ=0.9873, center=<1.8908 0.1157 -0.0515>}]
 strawFront_mstr [{shapeType=Box, extentX=1.01, extentY=1.01, extentZ=1.01}]
 strawMid1 [{shapeType=Box, extentX=1.01, extentY=1.01, extentZ=1.01, center=<0.2483 0.02 -0.034>}]
 strawMid2 [{shapeType=Box, extentX=3.01, extentY=1.01, extentZ=1.01, center=<1.5423 -0.0002 0.0028>}]
 strawMid3 [{shapeType=Box, extentX=2.51, extentY=1.01, extentZ=1.01, center=<1.2957 -0.0099 0.0015>}]
 strawMid4 [{shapeType=Box, extentX=4.2571, extentY=0.7509, extentZ=1.1879, center=<2.0955 -0.1139 -0.0359>}]
 strawMid_mstr [{shapeType=Box, extentX=1.01, extentY=1.01, extentZ=1.01}]
}
constraints [
 {bone1Name=strawFront1, bone2Name=strawFront_mstr, swing1LimitDegrees=12, swing2LimitDegrees=2, twistLimitDegrees=1}
 {bone1Name=strawFront2, bone2Name=strawFront1, swing1LimitDegrees=12, swing2LimitDegrees=2, twistLimitDegrees=1}
 {bone1Name=strawFront3, bone2Name=strawFront2, swing1LimitDegrees=12, swing2LimitDegrees=2, twistLimitDegrees=1}
 {bone1Name=strawFront4, bone2Name=strawFront3, swing1LimitDegrees=12, swing2LimitDegrees=2, twistLimitDegrees=1}
 {bone1Name=strawMid1, bone2Name=strawMid_mstr, swing1LimitDegrees=12, swing2LimitDegrees=2, twistLimitDegrees=1}
 {bone1Name=strawMid2, bone2Name=strawMid1, swing1LimitDegrees=12, swing2LimitDegrees=2, twistLimitDegrees=1}
 {bone1Name=strawMid3, bone2Name=strawMid2, swing1LimitDegrees=12, swing2LimitDegrees=2, twistLimitDegrees=1}
 {bone1Name=strawMid4, bone2Name=strawMid3, swing1LimitDegrees=12, swing2LimitDegrees=2, twistLimitDegrees=1}
 {bone1Name=strawBack1, bone2Name=strawBack_mstr, swing1LimitDegrees=12, swing2LimitDegrees=2, twistLimitDegrees=1}
 {bone1Name=strawBack2, bone2Name=strawBack1, swing1LimitDegrees=12, swing2LimitDegrees=2, twistLimitDegrees=1}
 {bone1Name=strawBack3, bone2Name=strawBack2, swing1LimitDegrees=12, swing2LimitDegrees=2, twistLimitDegrees=1}
 {bone1Name=strawBack4, bone2Name=strawBack3, swing1LimitDegrees=12, swing2LimitDegrees=2, twistLimitDegrees=1}
]
