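# Animation Blueprint for the Anarchy Acres Farmer back bling, giving the backpack a constant engine-like shake with physics on top.
# Each update, sine waves of game time at differing frequencies set small offsets for the pack's position and rotation and for the two exhaust stacks.
# Additive bone-space Modify Bone nodes apply them to pack, stackFront and stackBack.
# A RigidBody node then simulates in pack bone space with scaled-down component motion and overridden gravity.
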AnimBlueprint /BRCosmetics/Accessories/FORT_Backpacks/Backpack_M_MED_Anarchy_Acres_Farmer/Meshes/M_MED_Anarchy_Acres_Farmer_AnimBP
target_skeleton=&/M_MED_Anarchy_Acres_Farmer_Skeleton
graphs [
 AnimGraph {nodes [
   n1 = "Transform (Modify) Bone - Bone: pack"(
    ComponentPose=Local To Component(LocalPose=Local Space Ref Pose())
    Translation=Get backpackTranslate() -> (backpackTranslate:Vector)
    Rotation=Get backpackRotate() -> (backpackRotate:Rotator)
   ) {
    rotationMode=BMM_Additive
    rotationSpace=BCS_BoneSpace
    translationMode=BMM_Additive
    translationSpace=BCS_BoneSpace
    alphaBoolBlend {blendOption=Linear}
    boneToModify {boneName=pack}
   }
   n2 = "Transform (Modify) Bone - Bone: stackFront"(ComponentPose=@n1, Alpha=1, Translation=Get exhaustFront() -> (exhaustFront:Vector), Rotation="0,0,0", Scale="1,1,1") {
    rotationMode=BMM_Additive
    rotationSpace=BCS_BoneSpace
    translationMode=BMM_Additive
    translationSpace=BCS_BoneSpace
    alphaBoolBlend {blendOption=Linear}
    boneToModify {boneName=stackFront}
   }
   n3 = "Transform (Modify) Bone - Bone: stackBack"(ComponentPose=@n2, Alpha=1, Translation=Get exhaustBack() -> (exhaustBack:Vector), Rotation="0,0,0", Scale="1,1,1") {
    rotationMode=BMM_Additive
    rotationSpace=BCS_BoneSpace
    translationMode=BMM_Additive
    translationSpace=BCS_BoneSpace
    alphaBoolBlend {blendOption=Linear}
    boneToModify {boneName=stackBack}
   }
   n4 = RigidBody(ComponentPose=@n3) {
    bClampLinearTranslationLimitToRefPose=true
    bOverrideWorldGravity=true
    overridePhysicsAsset=&/M_MED_Anarchy_Acres_Farmer_Physics
    simulationSpace=BaseBoneSpace
    alphaBoolBlend {blendOption=Linear}
    baseBoneRef {boneName=pack}
    componentAppliedLinearAccClamp=<3000 3000 3000>
    componentLinearAccScale=<0.4 0.3 0.3>
    componentLinearVelScale=<0.4 0.3 0.3>
    overrideWorldGravity {z=-980}
   }
   Output Pose(Result=Component To Local(ComponentPose=@n4)) {layerGroup=None}
  ]}
 EventGraph {nodes [
   Event BlueprintUpdateAnimation()
   n1 = GetGameTimeInSeconds()
   n2 = "Sin (Radians)"(A=float * float(A=float * float(A=1, B=35), B=@n1))
   > Set exhaustFront(exhaustFront:Vector=MakeVector(X=float * float(A=@n2, B=0.25), Y=0)) -> (Output_Get:Vector)
   n3 = "Sin (Radians)"(A=float * float(A=float * float(A=1, B=30), B=@n1))
   > Set exhaustBack(exhaustBack:Vector=MakeVector(X=float * float(A=@n3, B=0.25), Y=0)) -> (Output_Get:Vector)
   n4 = "Sin (Radians)"(A=float * float(A=float * float(A=1, B=45), B=@n1))
   > Set backpackTranslate(backpackTranslate:Vector=MakeVector(Y=0, Z=float * float(A=@n4, B=0.09))) -> (Output_Get:Vector)
   n5 = "Sin (Radians)"(A=float * float(A=float * float(A=1, B=18), B=@n1))
   n6 = "Sin (Radians)"(A=float * float(A=float * float(A=1, B=20), B=@n1))
   > Set backpackRotate(backpackRotate:Rotator=MakeRotator(Pitch=float * float(A=@n5, B=0.2), Yaw=float * float(A=@n6, B=0.1))) -> (Output_Get:Rotator)
   TryGetPawnOwner()
  ]}
]
variables [
 backpackRotate {type=Rotator}
 backpackTranslate {type=Vector}
 exhaustBack {type=Vector}
 exhaustFront {type=Vector}
]
